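# Sound cue for Rita Repulsa's attack call-to-action voice lines in the ArrowTube content of ToyFoxGameplay.
# A Random node picks among four equally weighted wave players, each holding a different Rita line, such as urging the destruction of the Megazord.
# The cue plays through the narrator VO sound class at slightly reduced volume with a very large max distance.
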SoundCue /ToyFoxGameplay/Assets/Sound/VO/Rita/VO_ArrowTube_RitaRepulsa_AttackCTA
root=Random_1
nodes [
 {id=Random_1, children [WavePlayer_2,WavePlayer_4,WavePlayer_5,WavePlayer_6], properties {weights [1,1,1,1]}}
 {id=WavePlayer_2, properties {soundWaveAssetPtr=/ToyFoxGameplay/Assets/Sound/VO/Rita/Wavs/VO_ArrowTube_RitaRepulsa_TogetherWe_010}}
 {id=WavePlayer_4, properties {soundWaveAssetPtr=/ToyFoxGameplay/Assets/Sound/VO/Rita/Wavs/VO_ArrowTube_RitaRepulsa_DestroyMegazord_010}}
 {id=WavePlayer_5, properties {soundWaveAssetPtr=/ToyFoxGameplay/Assets/Sound/VO/Rita/Wavs/VO_ArrowTube_RitaRepulsa_DestroyThat_010}}
 {id=WavePlayer_6, properties {soundWaveAssetPtr=/ToyFoxGameplay/Assets/Sound/VO/Rita/Wavs/VO_ArrowTube_RitaRepulsa_AGiftAwaits_010}}
]
properties {maxDistance=2097152, soundClassObject=/Game/Sounds/FortSoundClassesAndModes/Fort_VO_Narrator, volumeMultiplier=0.7079}
