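# Sound cue for Rita Repulsa's battle-encouragement voice lines in the ArrowTube content.
# A Random node picks one of five equally weighted wave players, each holding a short encouragement line such as "Keep Fighting" or "Don't Let Up".
# The cue routes through the narrator VO sound class with slightly reduced volume and a very large max distance.
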SoundCue /ToyFoxGameplay/Assets/Sound/VO/Rita/VO_ArrowTube_RitaRepulsa_BattleEncourage
root=Random_0
nodes [
 {id=Random_0, children [WavePlayer_0,WavePlayer_1,WavePlayer_2,WavePlayer_3,WavePlayer_4], properties {weights [1,1,1,1,1]}}
 {id=WavePlayer_0, properties {soundWaveAssetPtr=/ToyFoxGameplay/Assets/Sound/VO/Rita/Wavs/VO_ArrowTube_RitaRepulsa_DontLetUp_010}}
 {id=WavePlayer_1, properties {soundWaveAssetPtr=/ToyFoxGameplay/Assets/Sound/VO/Rita/Wavs/VO_ArrowTube_RitaRepulsa_HowBeautifully_010}}
 {id=WavePlayer_2, properties {soundWaveAssetPtr=/ToyFoxGameplay/Assets/Sound/VO/Rita/Wavs/VO_ArrowTube_RitaRepulsa_KeepFighting_010}}
 {id=WavePlayer_3, properties {soundWaveAssetPtr=/ToyFoxGameplay/Assets/Sound/VO/Rita/Wavs/VO_ArrowTube_RitaRepulsa_YesYes_010}}
 {id=WavePlayer_4, properties {soundWaveAssetPtr=/ToyFoxGameplay/Assets/Sound/VO/Rita/Wavs/VO_ArrowTube_RitaRepulsa_YesKeepFighting_010}}
]
properties {maxDistance=2097152, soundClassObject=/Game/Sounds/FortSoundClassesAndModes/Fort_VO_Narrator, volumeMultiplier=0.7079}
