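# Sound cue for Rita Repulsa's laugh voice line (ArrowTube), part of the ToyFox gameplay boss VO.
# A Random node picks one of four equally weighted laugh recordings.
# It plays louder than default through the boss-fight VOX sound class, using the ToyFox spawn-assemble VO attenuation.
# A bus send also feeds the boss VO envelope-follower audio bus, named for modulation use.
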SoundCue /ToyFoxGameplay/Assets/Sound/VO/Rita/VO_ArrowTube_RitaRepulsa_Laugh
root=Random_0
nodes [
 {id=Random_0, children [WavePlayer_0,WavePlayer_1,WavePlayer_2,WavePlayer_3], properties {weights [1,1,1,1]}}
 {id=WavePlayer_0, properties {soundWaveAssetPtr=/ToyFoxGameplay/Assets/Sound/VO/Rita/Wavs/VO_ArrowTube_RitaRepulsa_Laugh_007}}
 {id=WavePlayer_1, properties {soundWaveAssetPtr=/ToyFoxGameplay/Assets/Sound/VO/Rita/Wavs/VO_ArrowTube_RitaRepulsa_Laugh_008}}
 {id=WavePlayer_2, properties {soundWaveAssetPtr=/ToyFoxGameplay/Assets/Sound/VO/Rita/Wavs/VO_ArrowTube_RitaRepulsa_Laugh_010}}
 {id=WavePlayer_3, properties {soundWaveAssetPtr=/ToyFoxGameplay/Assets/Sound/VO/Rita/Wavs/VO_ArrowTube_RitaRepulsa_Laugh_011}}
]
properties {
 attenuationSettings=/ToyFoxGameplay/Assets/Sound/ATT_Toyfox_Spawn_Assemble_VO
 maxDistance=200000
 soundClassObject=/Game/Sounds/FortSoundClassesAndModes/Fort_BossFight_SoundClass/Fort_BossFight_VOX
 volumeMultiplier=1.5
 busSends [{audioBus=/Game/Sounds/Modulation/EnvelopeFollowers/BR_Boss/AB_BR_Boss_VO_EnvFollow, minSendLevel=0, sendLevel=0.35, soundSourceBus=null}]
}
