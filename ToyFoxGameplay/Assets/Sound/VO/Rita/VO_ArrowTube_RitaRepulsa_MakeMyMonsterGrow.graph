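# Sound cue for Rita Repulsa's "Make My Monster Grow" ArrowTube voice line.
# A single wave player feeds the output directly, with no randomization or attenuation node.
# It uses the boss-fight VOX sound class with a boosted volume and a very large max distance.
# A partial bus send feeds the BR boss VO envelope follower audio bus.
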SoundCue /ToyFoxGameplay/Assets/Sound/VO/Rita/VO_ArrowTube_RitaRepulsa_MakeMyMonsterGrow
root=WavePlayer_1
nodes [{id=WavePlayer_1, properties {soundWaveAssetPtr=/ToyFoxGameplay/Assets/Sound/VO/Rita/Wavs/VO_ArrowTube_RitaRepulsa_MakeMy_010}}]
properties {
 maxDistance=2097152
 soundClassObject=/Game/Sounds/FortSoundClassesAndModes/Fort_BossFight_SoundClass/Fort_BossFight_VOX
 volumeMultiplier=1.7
 busSends [
  {audioBus=null, minSendLevel=0, sendLevel=1, soundSourceBus=null}
  {audioBus=/Game/Sounds/Modulation/EnvelopeFollowers/BR_Boss/AB_BR_Boss_VO_EnvFollow, minSendLevel=0, sendLevel=0.35, soundSourceBus=null}
 ]
 soundSubmixSends [{minSendLevel=0, soundSubmix=null}]
}
